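# Animation Blueprint for the Cosmos carrier backpack, hovering beside the player in match and front end.
# State machines play float loops, with additive layers for door open/close and turning toward the player during the pet emote montage.
# Event graph selects per-state offsets from player anim state, spring-interpolates them with velocity and melee twist, then drives hover, root and pack bones.
# Front end reads skydiving from the pet AnimBP; DBNO carry shrinks the root.
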AnimBlueprint /BRCosmetics/Accessories/FORT_Backpacks/Backpack_M_MED_Carrier_Cosmos/Animations/M_MED_Carrier_Cosmos_Pack_AnimBP
target_skeleton=/BRCosmetics/Accessories/FORT_Backpacks/Backpack_M_MED_Carrier_Cosmos/Meshes/M_MED_Carrier_Cosmos_Pack_Skeleton
graphs [
 AnimGraph {nodes [
   AnimGraphNode_StateMachine_0 = Carrier_Motions_SM() {is=title AnimGraphNode_StateMachine}
   n1 = Use cached pose 'TurnTowardsPose'()
   n2 = Get CarrierPackOffsetTotal() -> (CarrierPackOffsetTotal:Vector)
   n3 = "Transform (Modify) Bone - Bone: hover"(
    ComponentPose=Local To Component(LocalPose=Apply Additive(Base=@AnimGraphNode_StateMachine_0, Additive=@n1) {alphaBoolBlend {blendOption=Linear}})
    Alpha=1
    Translation=@n2
    Rotation="0,0,0"
    Scale="1,1,1"
   ) {translationMode=BMM_Additive, alphaBoolBlend {blendOption=Linear}, boneToModify {boneName=hover}}
   n4 = Get RootYawOffsetSmooth() -> (RootYawOffsetSmooth:double)
   n5 = "Transform (Modify) Bone - Bone: hover"(
    ComponentPose="Transform (Modify) Bone - Bone: pack"!(
     ComponentPose="Transform (Modify) Bone - Bone: root"!(ComponentPose=@n3, Alpha, Translation="0,0,0", Rotation_Roll:float=0, Rotation_Pitch:float=0, Rotation_Yaw:float=@n4, Scale) -> (Pose) {rotationMode=BMM_Additive, alphaBoolBlend {blendOption=Linear}, boneToModify {boneName=root}}
     Alpha
     Translation="0,0,0"
     Rotation_Roll:float=0
     Rotation_Pitch:float=0
     Rotation_Yaw:float=Get RootRotationSpring() -> (RootRotationSpring:double)
     Scale
    ) -> (Pose) {rotationMode=BMM_Additive, alphaBoolBlend {blendOption=Linear}, boneToModify {boneName=pack}}
    Alpha=1
    Translation="0,0,0"
    Rotation=Combine Rotators!(
     A:Rotator=Get CarrierPackOffsetRotationReaction() -> (CarrierPackOffsetRotationReaction:Rotator)
     B:Rotator=Get CarrierPackVelocityRotationReaction() -> (CarrierPackVelocityRotationReaction:Rotator)
    ) -> (ReturnValue:Rotator)
    Scale="1,1,1"
   ) {rotationMode=BMM_Additive, translationSpace=BCS_WorldSpace, alphaBoolBlend {blendOption=Linear}, boneToModify {boneName=hover}}
   n6 = "Transform (Modify) Bone - Bone: root"!(
    ComponentPose=@n5
    Alpha=Get IsDBNOCarriedAlpha() -> (IsDBNOCarriedAlpha:double)
    Translation="0,0,0"
    Rotation:Rotator="0,0,0"
    Scale="0.01,0.01,0.01"
   ) -> (Pose) {
    scaleMode=BMM_Replace
    scaleSpace=BCS_WorldSpace
    translationMode=BMM_Replace
    alphaBoolBlend {blendOption=Linear}
    boneToModify {boneName=root}
    scale=<0.01 0.01 0.01>
   }
   n7 = Component To Local(ComponentPose=@n6)
   AnimGraphNode_StateMachine_3 = FrontEnd() {is=title AnimGraphNode_StateMachine}
   n8 = Apply Additive(Base=@AnimGraphNode_StateMachine_3, Additive=Use cached pose 'TurnTowardsPose'()) {alphaBoolBlend {blendOption=Linear}}
   n9 = Component To Local(ComponentPose="Transform (Modify) Bone - Bone: hover"(ComponentPose=Local To Component(LocalPose=@n8), Alpha=1, Translation=Get CarrierPackOffsetTotal(), Rotation="0,0,0", Scale="1,1,1") {translationMode=BMM_Additive, alphaBoolBlend {blendOption=Linear}, boneToModify {boneName=hover}})
   n10 = Get bIsInFrontEnd()
   AnimGraphNode_StateMachine_1 = OpenCloseADD() {is=title AnimGraphNode_StateMachine}
   Output Pose(Result=Apply Additive(
     Base=Blend Poses by bool(BlendPose_0=@n9, BlendPose_1=@n7, BlendTime_0=0, BlendTime_1=0, bActiveValue=@n10) {blendType=Linear}
     Additive=@AnimGraphNode_StateMachine_1
    ) {alphaBoolBlend {blendOption=Linear}}) {layerGroup=None}
   AnimGraphNode_StateMachine_2 = TurnTowardsPlayer() {is=title AnimGraphNode_StateMachine}
   Save cached pose 'TurnTowardsPose'(Pose=@AnimGraphNode_StateMachine_2) {cacheName=TurnTowardsPose}
   vector + vector(A:Vector=@n2, B:Vector=Get ParachuteJustOpenedOffset() -> (ParachuteJustOpenedOffset:Vector)) -> (ReturnValue:Vector) {is=title K2Node_CommutativeAssociativeBinaryOperator}
  ]}
 Carrier_Motions_SM {
  entry=Floating
  path=AnimGraph.AnimGraphNode_StateMachine_0.Carrier_Motions_SM
  ids=id
  nodes [AnimStateNode_0 = Floating() {is=title AnimStateNode}]
 }
 Floating {
  path=AnimGraph.AnimGraphNode_StateMachine_0.Carrier_Motions_SM.AnimStateNode_0.Floating
  nodes [
   n1 = Sequence Player 'Cosmos_Carrier_Float'() {sequence=&/Cosmos_Carrier_Float}
   n2 = Sequence Player 'Cosmos_Carrier_Float_Loco'() {sequence=&/Cosmos_Carrier_Float_Loco}
   n3 = Get FortPlayerAnimBP() -> (FortPlayerAnimBP:Fortnite M Avg Player Anim Blueprint Object Reference)
   n4 = OR Boolean(
    A=Get bIsMoving2D!(self:Fort Player Anim Instance Object Reference=@n3) -> (bIsMoving2D)
    B=Get bIsPlayingMeleeAnim!(self:Fort Player Anim Instance Object Reference=@n3) -> (bIsPlayingMeleeAnim)
   )
   n5 = SelectFloat(A=1, B=1.5, bPickA=Get IsParachuteOpen() -> (IsParachuteOpen:bool))
   Output Animation Pose(Result=Blend Poses by bool(
     BlendPose_0=Sequence Player 'Cosmos_Carrier_Float_Skydive'!(PlayRate:float=@n5) -> (Pose) {sequence=&/Cosmos_Carrier_Float_Skydive}
     BlendPose_1=Blend Poses by bool(BlendPose_0=@n2, BlendPose_1=@n1, bActiveValue=@n4) {blendType=Linear}
     bActiveValue=Get IsSkydiving() -> (IsSkydiving:bool)
    ) {blendType=Linear})
  ]
 }
 OpenCloseADD {
  entry=OpenLoop
  path=AnimGraph.AnimGraphNode_StateMachine_1.OpenCloseADD
  ids=id
  nodes [
   AnimStateNode_0 = OpenLoop() {is=title AnimStateNode}
   AnimStateNode_1 = CloseDoors() {is=title AnimStateNode}
   AnimStateNode_2 = ClosedLoop() {is=title AnimStateNode}
   AnimStateNode_3 = OpenDoors() {is=title AnimStateNode}
   AnimStateTransitionNode_0 = AnimStateTransitionNode() {crossfadeDuration=0}
   AnimStateTransitionNode_1 = AnimStateTransitionNode() {bAutomaticRuleBasedOnSequencePlayerInState=true, crossfadeDuration=0}
   AnimStateTransitionNode_2 = AnimStateTransitionNode() {crossfadeDuration=0}
   AnimStateTransitionNode_3 = AnimStateTransitionNode() {bAutomaticRuleBasedOnSequencePlayerInState=true, crossfadeDuration=0}
  ]
  links [
   {
    dst=ClosedLoop
    rule=AnimGraph.AnimGraphNode_StateMachine_1.OpenCloseADD.AnimStateTransitionNode_1.Transition
    src=CloseDoors
    via=AnimStateTransitionNode_1
   }
   {
    dst=OpenDoors
    rule=AnimGraph.AnimGraphNode_StateMachine_1.OpenCloseADD.AnimStateTransitionNode_2.Transition
    src=ClosedLoop
    via=AnimStateTransitionNode_2
   }
   {
    dst=OpenLoop
    rule=AnimGraph.AnimGraphNode_StateMachine_1.OpenCloseADD.AnimStateTransitionNode_3.Transition
    src=OpenDoors
    via=AnimStateTransitionNode_3
   }
   {
    dst=CloseDoors
    rule=AnimGraph.AnimGraphNode_StateMachine_1.OpenCloseADD.AnimStateTransitionNode_0.Transition
    src=OpenLoop
    via=AnimStateTransitionNode_0
   }
  ]
 }
 OpenLoop {
  path=AnimGraph.AnimGraphNode_StateMachine_1.OpenCloseADD.AnimStateNode_0.OpenLoop
  nodes [
   Additive Identity Pose()
   Output Animation Pose(Result="Sequence Player (additive) 'Cosmos_Carrier_Open_Loop'"() {sequence=&/Cosmos_Carrier_Open_Loop})
  ]
 }
 CloseDoors {
  path=AnimGraph.AnimGraphNode_StateMachine_1.OpenCloseADD.AnimStateNode_1.CloseDoors
  nodes [Output Animation Pose(Result="Sequence Player (additive) 'Cosmos_Carrier_Close'"!(bLoopAnimation:bool=false) -> (Pose) {sequence=&/Cosmos_Carrier_Close})]
 }
 ClosedLoop {
  path=AnimGraph.AnimGraphNode_StateMachine_1.OpenCloseADD.AnimStateNode_2.ClosedLoop
  nodes [Output Animation Pose(Result="Sequence Player (additive) 'Cosmos_Carrier_Close_Loop'"() {sequence=&/Cosmos_Carrier_Close_Loop})]
 }
 OpenDoors {
  path=AnimGraph.AnimGraphNode_StateMachine_1.OpenCloseADD.AnimStateNode_3.OpenDoors
  nodes [Output Animation Pose(Result="Sequence Player (additive) 'Cosmos_Carrier_Open'"!(bLoopAnimation:bool=false) -> (Pose) {sequence=&/Cosmos_Carrier_Open})]
 }
 Transition {
  path=AnimGraph.AnimGraphNode_StateMachine_1.OpenCloseADD.AnimStateTransitionNode_0.Transition
  nodes [Result(bCanEnterTransition=Get bCarrierDoorsClose() -> (bCarrierDoorsClose:bool))]
 }
 Transition {path=AnimGraph.AnimGraphNode_StateMachine_1.OpenCloseADD.AnimStateTransitionNode_1.Transition, nodes [Result()]}
 Transition {
  path=AnimGraph.AnimGraphNode_StateMachine_1.OpenCloseADD.AnimStateTransitionNode_2.Transition
  nodes [Result(bCanEnterTransition=NOT Boolean(A=Get bCarrierDoorsClose()))]
 }
 Transition {path=AnimGraph.AnimGraphNode_StateMachine_1.OpenCloseADD.AnimStateTransitionNode_3.Transition, nodes [Result()]}
 TurnTowardsPlayer {
  entry=Identity
  path=AnimGraph.AnimGraphNode_StateMachine_2.TurnTowardsPlayer
  ids=id
  nodes [
   AnimStateNode_0 = Identity() {is=title AnimStateNode}
   AnimStateNode_1 = TurnTowardsPlayerStart() {is=title AnimStateNode}
   AnimStateNode_2 = TurnTowardsPlayerLoop() {is=title AnimStateNode}
   AnimStateNode_3 = TurnTowardsPlayerEnd() {is=title AnimStateNode}
   AnimStateTransitionNode_0 = AnimStateTransitionNode() {crossfadeDuration=0}
   AnimStateTransitionNode_1 = AnimStateTransitionNode() {bAutomaticRuleBasedOnSequencePlayerInState=true, crossfadeDuration=0}
   AnimStateTransitionNode_2 = AnimStateTransitionNode() {crossfadeDuration=0}
   AnimStateTransitionNode_3 = AnimStateTransitionNode() {bAutomaticRuleBasedOnSequencePlayerInState=true, crossfadeDuration=0}
  ]
  links [
   {
    dst=TurnTowardsPlayerStart
    rule=AnimGraph.AnimGraphNode_StateMachine_2.TurnTowardsPlayer.AnimStateTransitionNode_0.Transition
    src=Identity
    via=AnimStateTransitionNode_0
   }
   {
    dst=Identity
    rule=AnimGraph.AnimGraphNode_StateMachine_2.TurnTowardsPlayer.AnimStateTransitionNode_3.Transition
    src=TurnTowardsPlayerEnd
    via=AnimStateTransitionNode_3
   }
   {
    dst=TurnTowardsPlayerEnd
    rule=AnimGraph.AnimGraphNode_StateMachine_2.TurnTowardsPlayer.AnimStateTransitionNode_2.Transition
    src=TurnTowardsPlayerLoop
    via=AnimStateTransitionNode_2
   }
   {
    dst=TurnTowardsPlayerLoop
    rule=AnimGraph.AnimGraphNode_StateMachine_2.TurnTowardsPlayer.AnimStateTransitionNode_1.Transition
    src=TurnTowardsPlayerStart
    via=AnimStateTransitionNode_1
   }
  ]
 }
 Identity {
  path=AnimGraph.AnimGraphNode_StateMachine_2.TurnTowardsPlayer.AnimStateNode_0.Identity
  nodes [Output Animation Pose(Result=Additive Identity Pose())]
 }
 TurnTowardsPlayerStart {
  path=AnimGraph.AnimGraphNode_StateMachine_2.TurnTowardsPlayer.AnimStateNode_1.TurnTowardsPlayerStart
  nodes [Output Animation Pose(Result="Sequence Player (additive) 'Cosmos_Carrier_TurnTowardsPlayer_Start'"!(bLoopAnimation:bool=false) -> (Pose) {sequence=&/Cosmos_Carrier_TurnTowardsPlayer_Start})]
 }
 TurnTowardsPlayerLoop {
  path=AnimGraph.AnimGraphNode_StateMachine_2.TurnTowardsPlayer.AnimStateNode_2.TurnTowardsPlayerLoop
  nodes [Output Animation Pose(Result="Sequence Player (additive) 'Cosmos_Carrier_TurnTowardsPlayer_Loop'"() {sequence=&/Cosmos_Carrier_TurnTowardsPlayer_Loop})]
 }
 TurnTowardsPlayerEnd {
  path=AnimGraph.AnimGraphNode_StateMachine_2.TurnTowardsPlayer.AnimStateNode_3.TurnTowardsPlayerEnd
  nodes [Output Animation Pose(Result="Sequence Player (additive) 'Cosmos_Carrier_TurnTowardsPlayer_End'"!(bLoopAnimation:bool=false) -> (Pose) {sequence=&/Cosmos_Carrier_TurnTowardsPlayer_End})]
 }
 Transition {
  path=AnimGraph.AnimGraphNode_StateMachine_2.TurnTowardsPlayer.AnimStateTransitionNode_0.Transition
  nodes [Result(bCanEnterTransition=Get Emote_NonPlayerTrigger Montage Playing() -> (Emote_NonPlayerTrigger Montage Playing:bool))]
 }
 Transition {path=AnimGraph.AnimGraphNode_StateMachine_2.TurnTowardsPlayer.AnimStateTransitionNode_1.Transition, nodes [Result()]}
 Transition {
  path=AnimGraph.AnimGraphNode_StateMachine_2.TurnTowardsPlayer.AnimStateTransitionNode_2.Transition
  nodes [Result(bCanEnterTransition=Get Emote_NonPlayerTrigger Montage Playing())]
 }
 Transition {path=AnimGraph.AnimGraphNode_StateMachine_2.TurnTowardsPlayer.AnimStateTransitionNode_3.Transition, nodes [Result()]}
 FrontEnd {
  entry=Idle
  path=AnimGraph.AnimGraphNode_StateMachine_3.FrontEnd
  ids=id
  nodes [
   AnimStateNode_0 = Idle()
   AnimStateNode_1 = Skydiving() {is=title AnimStateNode}
   AnimStateTransitionNode_0 = AnimStateTransitionNode()
   AnimStateTransitionNode_1 = AnimStateTransitionNode()
  ]
  links [
   {
    dst=Skydiving
    rule=AnimGraph.AnimGraphNode_StateMachine_3.FrontEnd.AnimStateTransitionNode_0.Transition
    src=Idle
    via=AnimStateTransitionNode_0
   }
   {
    dst=Idle
    rule=AnimGraph.AnimGraphNode_StateMachine_3.FrontEnd.AnimStateTransitionNode_1.Transition
    src=Skydiving
    via=AnimStateTransitionNode_1
   }
  ]
 }
 Idle {
  path=AnimGraph.AnimGraphNode_StateMachine_3.FrontEnd.AnimStateNode_0.Idle
  nodes [Output Animation Pose(Result=Sequence Player 'Cosmos_Carrier_Float'() {sequence=&Cosmos_Carrier_Float})]
 }
 Skydiving {
  path=AnimGraph.AnimGraphNode_StateMachine_3.FrontEnd.AnimStateNode_1.Skydiving
  nodes [Output Animation Pose(Result=Sequence Player 'Cosmos_Carrier_Float_Loco'() {sequence=&Cosmos_Carrier_Float_Loco})]
 }
 Transition {
  path=AnimGraph.AnimGraphNode_StateMachine_3.FrontEnd.AnimStateTransitionNode_0.Transition
  nodes [Result(bCanEnterTransition=Get IsSkydiving())]
 }
 Transition {
  path=AnimGraph.AnimGraphNode_StateMachine_3.FrontEnd.AnimStateTransitionNode_1.Transition
  nodes [Result(bCanEnterTransition=NOT Boolean(A=Get IsSkydiving()))]
 }
 EventGraph {nodes [
   CarrierDoorsClose() {is=title K2Node_CustomEvent}
   > Set bCarrierDoorsClose(bCarrierDoorsClose:bool=true) -> (Output_Get:bool)
   CarrierDoorsOpen() {is=title K2Node_CustomEvent}
   > Set bCarrierDoorsClose(bCarrierDoorsClose=false)
   n1 = Event BlueprintUpdateAnimation()
   > Set DeltaTimeX(DeltaTimeX:double=@n1.DeltaTimeX) -> (Output_Get:double)
   > n2 = Branch(Condition=Get bIsInFrontEnd())
   > K2Node_Composite_3 = FrontEndGraph() {is=title K2Node_Composite}
   n3 = Sequence!(execute=@n2.else) -> (then_0, then_1, then_2:Exec)
   > n4 = Set FortPlayerAnimBP(FortPlayerAnimBP:Fortnite M Avg Player Anim Blueprint Object Reference=Cast To Fortnite_M_Avg_Player_AnimBlueprint!(Object=GetOwnerFortPlayerAnimInstance!(self:Custom Character Part Anim Instance Object Reference) -> (ReturnValue:Fort Player Anim Instance Object Reference)) -> (AsFortnite M Avg Player Anim Blueprint:Fortnite M Avg Player Anim Blueprint Object Reference, bSuccess:bool=~) {is=title K2Node_DynamicCast}) -> (Output_Get:Fortnite M Avg Player Anim Blueprint Object Reference)
   > Is Valid(InputObject=@n4)
   n5 = Get FortPlayerAnimBP()
   n6 = Get!(self:Fort Player Anim Instance Object Reference=@n5) -> (LocalVelocity_X:double, LocalVelocity_Y:double, LocalVelocity_Z:double)
   > Set FortPlayerLocalVelocity(FortPlayerLocalVelocity:Vector=MakeVector(X=@n6, Y=@n6.LocalVelocity_Y, Z=Get PawnVelocityZ!(self:Fort Anim Instance Object Reference=@n5) -> (PawnVelocityZ))) -> (Output_Get:Vector)
   n7 = Break Fort Anim Input Skydiving!(FortAnimInput_Skydiving:"Fort Anim Input Skydiving Structure (by ref)"=Get Skydiving!(self:Fort Player Anim Instance Object Reference=Get FortPlayerAnimBP()) -> (Skydiving:Fort Anim Input Skydiving Structure)) -> (
    ParaGilderRootOffset:Vector=~
    ParaGliderRootRotationOverride:Rotator=~
    bUseParaGlideRootModifier:bool=~
    bIsSkydivingFromLaunchPad:bool=~
    bIsSkydivingFromBus:bool=~
    bIsInVortex:bool=~
    bIsDiving:bool=~
    bIsParachuteOpen:bool=~
    bIsSkydiving:bool=~
    bIsParachuteLeaning:bool=~
    bIsSkydiveLeaning:bool=~
    bIsLeaning:bool=~
    bIsSkydiveDiveMode:bool=~
    bPlaySkydiveDrift:bool=~
    bSkydiveDriftDelayActive:bool=~
    bSkydiveDriftAnimAllowed:bool=~
    bIsGliderRight:bool=~
    bIsGliderCenter:bool=~
    bIsGliderLeft:bool=~
    bIsGliderForward:bool=~
    bIsGliderBack:bool=~
    bHasDeployedAndIsStandingUpright:bool=~
    LocalAccelForward:float=~
    LocalAccelRight:float=~
    LocalVelocityRight:float=~
    SkydiveAimPitch:float=~
    SkydiveAimYaw:float=~
    DeployChuteAnimRate:float=~
    SkydiveDriftAnimRate:float=~
    SkydiveDriftAnimRateCurrent:float=~
    SkydiveFidgetAnimRate:float=~
    SkydiveFidgetAnimRateCurrent:float=~
    SkydiveAdditiveAlpha:float=~
    SkydiveDriftDelay:float=~
    SkydiveDriftAnim:int
    SkydiveDriftAnimMax:int
    LaunchpadAnim:int
    LocalAccelDir:ESkydivingDirection Enum
    DirectionLast:ESkydivingDirection Enum
    PoseAdditive:Anim Sequence Object Reference
   ) {is=title K2Node_BreakStruct}
   > Set IsSkydiving(IsSkydiving:bool=@n7.bIsSkydiving) -> (Output_Get:bool)
   > Set IsParachuteOpen(IsParachuteOpen:bool=@n7.bIsParachuteOpen) -> (Output_Get:bool)
   > Set IsDiving(IsDiving:bool=@n7.bIsDiving) -> (Output_Get:bool)
   > n8 = Set Is DBNOCarried(Is DBNOCarried:bool=Get bIsDBNOCarried!(self:Fort Player Anim Instance Object Reference=Get FortPlayerAnimBP()) -> (bIsDBNOCarried)) -> (Output_Get:bool)
   > Set IsDBNOCarriedAlpha(IsDBNOCarriedAlpha:double="To Float (Boolean)"(InBool=@n8)) -> (Output_Get:double)
   n9 = GetTurnInPlaceInput!(self:Fort Player Animation Interface Object Reference=GetOwnerAnimationInterface!(self:Custom Character Part Anim Instance Object Reference) -> (ReturnValue:Fort Player Animation Interface Object Reference)) -> (ReturnValue:"Fort Anim Input Turn In Place Structure (by ref)")
   > n10 = Set RootYawOffset(RootYawOffset:double=float * float(
     A=Break Fort Anim Input Turn in Place!(FortAnimInput_TurnInPlace:"Fort Anim Input Turn In Place Structure (by ref)"=@n9) -> (
      RootYawOffset:float=~
      TurnAnimPlayRate:float=~
      bWantsToTurn:bool=~
      bExitTurn:bool=~
      bUseTurnA:bool=~
      TurnACurveAlpha:float=~
      TurnBCurveAlpha:float=~
      bWantsToTurnInVehicle:bool=~
      bWantsToTurnAgain:bool=~
      bTurningLeft:bool=~
      ShufflePlayRate:float=~
      CrouchShufflePlayRate:float=~
      CrouchShuffleStartPosition:float=~
      CrouchShuffleSpeedScaling:float=~
      bTransition_LowerBody_Shuffle:bool=~
      bTransition_LowerBody_ShuffleADS:bool=~
      bPlayOutro:bool=~
     ) {is=title K2Node_BreakStruct}
     B=0.5
    )) -> (Output_Get:double)
   > Set RootYawOffsetSmooth(RootYawOffsetSmooth:double=FInterpTo(Current=Get RootYawOffsetSmooth(), Target=@n10, DeltaTime=Get DeltaTimeX() -> (DeltaTimeX:double), InterpSpeed=2)) -> (Output_Get:double)
   > K2Node_Composite_4 = OpenParachuteOffset!(then:Exec) -> (then_2:Exec) {is=title K2Node_Composite}
   > K2Node_Composite_5 = Add an interped offset when jumping up to keep cape from clipping!(execute) -> (then:Exec) {is=title K2Node_Composite}
   > K2Node_Composite_1 = SpringValues!(execute) -> (then:Exec) {is=title K2Node_Composite}
   n11 = Get CarrierPackOffsetTotal()
   n12 = Get CarrierPackOffsetTotalLastTick() -> (CarrierPackOffsetTotalLastTick:Vector)
   > n13 = Set CarrierPackOffsetRate(CarrierPackOffsetRate:Vector=vector / float!(A:Vector=vector - vector!(A:Vector=@n12, B:Vector=@n11) -> (ReturnValue:Vector) {is=title K2Node_CallFunction}, B:double=Get DeltaTimeX()) -> (ReturnValue:Vector)) -> (Output_Get:Vector)
   > Set CarrierPackOffsetRateSmooth(CarrierPackOffsetRateSmooth:Vector=VInterpTo!(
     Current:Vector=Get CarrierPackOffsetRateSmooth() -> (CarrierPackOffsetRateSmooth:Vector)
     Target:Vector=@n13
     DeltaTime:float=Get DeltaTimeX()
     InterpSpeed:float=3
    ) -> (ReturnValue:Vector)) -> (Output_Get:Vector)
   n14 = BreakVector(InVec=Get CarrierPackOffsetRateSmooth())
   n15 = float * float(A=@n14, B=-0.2)
   n16 = float * float(A=@n14.Y, B=0.2)
   n17 = float * float(A=@n14.Z, B=-0.1)
   > Set CarrierPackOffsetRotationReaction(CarrierPackOffsetRotationReaction:Rotator=MakeRotator(Roll=float + float(A=@n16, B=@n17) {is=title K2Node_CommutativeAssociativeBinaryOperator}, Pitch=@n15, Yaw=@n15)) -> (Output_Get:Rotator)
   > K2Node_Composite_2 = MovingCarrierAwayFromEmotes!(execute) -> (then:Exec) {is=title K2Node_Composite}
   K2Node_Composite_0 = ValuesLastTick(execute=@n3.then_2) {is=title K2Node_Composite}
   Event BlueprintInitializeAnimation()
   > n18 = Cast To FortPlayerPawn(Object=TryGetPawnOwner())
   > Set FortPlayerPawn(FortPlayerPawn:Fort Player Pawn Object Reference=@n18) -> (Output_Get:Fort Player Pawn Object Reference)
   n19 = Get FortPlayerPawn() -> (FortPlayerPawn:Fort Player Pawn Object Reference)
   > n20 = Cast To Fortnite_M_Avg_Player_AnimBlueprint!(execute:Exec, Object=GetAnimInstance(self=Get Mesh(self=@n19))) -> (then:Exec, CastFailed:Exec, AsFortnite M Avg Player Anim Blueprint)
   > Set FortPlayerAnimBP(FortPlayerAnimBP=@n20)
   SelectFloat(A=0, B=2, bPickA=false)
   "float < float"!(
    A:double="Absolute (Float)"(A=Get RootYawOffsetLastTick() -> (RootYawOffsetLastTick:double))
    B:double="Absolute (Float)"(A=Get RootYawOffset() -> (RootYawOffset:double))
   ) -> (ReturnValue:bool) {is=title K2Node_CallFunction}
  ]}
 ValuesLastTick {
  path=EventGraph.K2Node_Composite_0.ValuesLastTick
  nodes [
   Inputs()
   > Set RootYawOffsetLastTick(RootYawOffsetLastTick:double=Get RootYawOffset()) -> (Output_Get:double)
   > Set CarrierPackOffsetTotalLastTick(CarrierPackOffsetTotalLastTick:Vector=Get CarrierPackOffsetTotal()) -> (Output_Get:Vector)
   > Set IsParachuteOpenLastTick(IsParachuteOpenLastTick:bool=Get IsParachuteOpen()) -> (Output_Get:bool)
   Outputs()
  ]
 }
 SpringValues {
  path=EventGraph.K2Node_Composite_1.SpringValues
  nodes [
   Inputs()
   > n1 = Sequence!(execute) -> (then_0, then_1)
   n2 = Get FortPlayerAnimBP()
   n3 = float * float(
    A=Get MeleeTwistCurveValue!(self:Fort Player Anim Instance Object Reference=@n2) -> (MeleeTwistCurveValue:float)
    B=SelectFloat(A=0.15, B=0.2, bPickA=Get bIsMoving2D(self=@n2))
   )
   n4 = Get!() -> (FortPlayerLocalVelocity_X:double, FortPlayerLocalVelocity_Y:double, FortPlayerLocalVelocity_Z:double)
   n5 = "Clamp (Float)"(Value=float * float(A=@n4.FortPlayerLocalVelocity_Y, B=0.03), Min=-30, Max=30)
   n6 = "Clamp (Float)"(Value=float * float(A=@n4.FortPlayerLocalVelocity_Z, B=-0.04), Min=-30, Max=30)
   n7 = "Clamp (Float)"(Value=float * float(A=@n4, B=-0.03), Min=-30, Max=2)
   n8 = VectorSpringInterp!(
    execute=@n1.then_1
    Current:Vector=Get CarrierPackOffsetTotal()
    Target:Vector=vector + vector!(
     A=Get Offset Values!() -> (Default Idle:Vector=~) {is=title K2Node_MacroInstance}
     B=MakeVector(X=@n5, Y=@n7, Z=@n6)
     C:Vector=MakeVector(X=@n3, Y=0)
     D:Vector=Get ParachuteJustOpenedOffset()
    ) -> (ReturnValue)
    SpringState:"Vector Spring State Structure (by ref)"=Get CarrierPackOffsetSpringState() -> (CarrierPackOffsetSpringState:Vector Spring State Structure)
    Stiffness:float=10
    CriticalDampingFactor:float=0.5
    DeltaTime:float=Get DeltaTimeX()
    Mass:float=0.75
    TargetVelocityAmount:float=1
    bClamp:bool=false
    MinValue:Vector="-1,-1,-1"
    MaxValue:Vector="1,1,1"
    bInitializeFromTarget:bool=false
   ) -> (then, ReturnValue:Vector)
   > Set CarrierPackOffsetTotal(CarrierPackOffsetTotal:Vector=@n8) -> (Output_Get:Vector)
   n9 = Get FortPlayerAnimBP()
   n10 = Get bIsMoving2D(self=@n9)
   > n11 = FloatSpringInterp!(
    execute
    Current:float=Get RootRotationSpring()
    Target:float=float * float(A=Get MeleeTwistCurveValue(self=@n9), B=SelectFloat(A=0.05, B=0.1, bPickA=@n10))
    SpringState:"Float Spring State Structure (by ref)"=Get CarrierPackRotationSpringState() -> (CarrierPackRotationSpringState:Float Spring State Structure)
    Stiffness:float=20
    CriticalDampingFactor:float=0.5
    DeltaTime:float=Get DeltaTimeX()
    Mass:float=0.75
    TargetVelocityAmount:float=1
    bClamp:bool=false
    MinValue:float=-1
    MaxValue:float=1
    bInitializeFromTarget:bool=false
   ) -> (then, ReturnValue:float)
   > Set RootRotationSpring(RootRotationSpring:double=@n11) -> (Output_Get:double)
   > Outputs!(then:Exec)
   n12 = Get FortPlayerLocalVelocity() -> (FortPlayerLocalVelocity:Vector)
   SelectFloat(A=50, B=10, bPickA="float > float"(A=Vector Length(A=@n12), B=400) {is=title K2Node_CallFunction})
   n13 = Get FortPlayerLocalVelocity()
   float * float(A=NormalizeToRange!(Value:double=Vector Length(A=@n13), RangeMin:double=0, RangeMax:double=500) -> (ReturnValue:double), B=10)
  ]
 }
 MovingCarrierAwayFromEmotes {
  path=EventGraph.K2Node_Composite_2.MovingCarrierAwayFromEmotes
  nodes [
   Inputs()
   > n1 = Branch(Condition=Get bIsPlayingEmote!(self:Fort Player Anim Instance Object Reference=Get FortPlayerAnimBP()) -> (bIsPlayingEmote))
   n2 = Get FortPlayerPawn()
   > n3 = Branch(Condition=IsNullItem!(ItemObject:Object Reference=GetCurrentMontage!(self:Character Object Reference=@n2) -> (ReturnValue:Anim Montage Object Reference)) -> (Is Null Item:bool))
   Set Emote_NonPlayerTrigger Montage Playing(execute=[@n3.then, @n1.else], Emote_NonPlayerTrigger Montage Playing:bool=false) -> (Output_Get:bool)
   > n4 = Set AnyEmotePlaying(AnyEmotePlaying:bool=false) -> (Output_Get:bool)
   Outputs(then=[@n6.else, @n4.then])
   Set AnyEmotePlaying(execute=@n3.else, AnyEmotePlaying=true)
   n5 = Get FortPlayerPawn()
   > n6 = Branch(Condition="Equal (Object)"(A=GetCurrentMontage(self=@n5), B=Get Emote_NonPlayerTrigger Montage() -> (Emote_NonPlayerTrigger Montage:Anim Montage Object Reference)) {is=title K2Node_CallFunction})
   > Set Emote_NonPlayerTrigger Montage Playing(Emote_NonPlayerTrigger Montage Playing=true)
  ]
 }
 FrontEndGraph {
  path=EventGraph.K2Node_Composite_3.FrontEndGraph
  nodes [
   Inputs()
   > n1 = Sequence!(execute) -> (then_0, then_1, then_2:Exec)
   > n2 = Cast To B_PetInst_Cosmos!(
    execute
    Object=Get CosmeticPetInstance!(self:Fort Player Pawn Object Reference=Get FortPlayerPawn()) -> (CosmeticPetInstance:Fort Player Pet Object Reference)
   ) -> (then, CastFailed, AsB Pet Inst Cosmos:B Pet Inst Cosmos Object Reference) {is=title K2Node_DynamicCast}
   > n3 = Set PetInstCosmos(PetInstCosmos:B Pet Inst Cosmos Object Reference=@n2) -> (Output_Get:B Pet Inst Cosmos Object Reference)
   > n4 = Cast To M_MED_Cosmos_Pet_AnimBP!(
    execute
    Object=GetAnimInstance(self=Get PetMesh!(self:Fort Player Pet Object Reference=@n3) -> (PetMesh:Skeletal Mesh Component Budgeted Object Reference))
   ) -> (then, CastFailed, AsM MED Cosmos Pet Anim BP:M MED Cosmos Pet Anim BP Object Reference) {is=title K2Node_DynamicCast}
   > Set PetAnimBP(PetAnimBP:M MED Cosmos Pet Anim BP Object Reference=@n4) -> (Output_Get:M MED Cosmos Pet Anim BP Object Reference)
   n5 = Get PetAnimBP() -> (PetAnimBP:M MED Cosmos Pet Anim BP Object Reference)
   > Set IsSkydiving(IsSkydiving=Break Fort Anim Input Skydiving(FortAnimInput_Skydiving=Get SkydivingState!(self:Fort Pet Anim Instance Object Reference=@n5) -> (SkydivingState:Fort Anim Input Skydiving Structure)) -> (
     ParaGilderRootOffset=~
     ParaGliderRootRotationOverride=~
     bUseParaGlideRootModifier=~
     bIsSkydivingFromLaunchPad=~
     bIsSkydivingFromBus=~
     bIsInVortex=~
     bIsDiving=~
     bIsParachuteOpen=~
     bIsSkydiving=~
     bIsParachuteLeaning=~
     bIsSkydiveLeaning=~
     bIsLeaning=~
     bIsSkydiveDiveMode=~
     bPlaySkydiveDrift=~
     bSkydiveDriftDelayActive=~
     bSkydiveDriftAnimAllowed=~
     bIsGliderRight=~
     bIsGliderCenter=~
     bIsGliderLeft=~
     bIsGliderForward=~
     bIsGliderBack=~
     bHasDeployedAndIsStandingUpright=~
     LocalAccelForward=~
     LocalAccelRight=~
     LocalVelocityRight=~
     SkydiveAimPitch=~
     SkydiveAimYaw=~
     DeployChuteAnimRate=~
     SkydiveDriftAnimRate=~
     SkydiveDriftAnimRateCurrent=~
     SkydiveFidgetAnimRate=~
     SkydiveFidgetAnimRateCurrent=~
     SkydiveAdditiveAlpha=~
     SkydiveDriftDelay=~
    ).bIsSkydiving)
   n6 = VectorSpringInterp(
    execute=@n1.then_1
    Current=Get CarrierPackOffsetTotal()
    Target=SelectVector(A=MakeVector(Y=45, Z=70), B=MakeVector(X=10, Y=-30), bPickA=Get IsSkydiving())
    SpringState=Get CarrierPackOffsetSpringState()
    Stiffness=10
    CriticalDampingFactor=0.5
    DeltaTime=Get DeltaTimeX()
    Mass=0.75
    TargetVelocityAmount=1
    bClamp=false
    MinValue="-1,-1,-1"
    MaxValue="1,1,1"
    bInitializeFromTarget=false
   )
   > Set CarrierPackOffsetTotal(CarrierPackOffsetTotal=@n6)
   n7 = Get FortPlayerPawn()
   Set Emote_NonPlayerTrigger Montage Playing(
    execute=@n1.then_2
    Emote_NonPlayerTrigger Montage Playing="Equal (Object)"(A=GetCurrentMontage(self=@n7), B=Get Emote_NonPlayerTrigger Montage())
   )
   Outputs!()
  ]
 }
 OpenParachuteOffset {
  path=EventGraph.K2Node_Composite_4.OpenParachuteOffset
  nodes [
   Inputs!() -> (then:Exec)
   > n1 = Sequence()
   > Branch(Condition=Get IsParachuteOpen())
   > n2 = Branch(Condition=Get IsParachuteOpenLastTick() -> (IsParachuteOpenLastTick:bool))
   Do Once(execute=@n2.else, Reset=@n2.then)
   > Set IsParachuteJustOpened(IsParachuteJustOpened:bool=true) -> (Output_Get:bool)
   > RetriggerableDelay(Duration=1.2)
   > Set IsParachuteJustOpened(IsParachuteJustOpened=false)
   Outputs!(then_2:Exec=@n1.then_2)
   n3 = Get IsParachuteJustOpened() -> (IsParachuteJustOpened:bool)
   Set ParachuteJustOpenedOffset(
    execute=@n1.then_1
    ParachuteJustOpenedOffset:Vector=VInterpTo(
     Current=Get ParachuteJustOpenedOffset()
     Target=SelectVector(A="0,20,30", B="0, 0, 0", bPickA=@n3)
     DeltaTime=Get DeltaTimeX()
     InterpSpeed=SelectFloat(A=3, B=1, bPickA=@n3)
    )
   ) -> (Output_Get:Vector)
   PrintString!(execute, InString="To String (Boolean)"(InBool=Get IsParachuteJustOpened()), bPrintToScreen, bPrintToLog, TextColor, Duration, Key) -> (then)
  ]
 }
 Add an interped offset when jumping up to keep cape from clipping {
  path=EventGraph.K2Node_Composite_5.Add an interped offset when jumping up to keep cape from clipping
  nodes [
   Inputs!() -> (execute:Exec)
   n1 = Get FortPlayerAnimBP()
   n2 = AND Boolean(
    A=Get bIsJumping!(self:Fort Anim Instance Object Reference=@n1) -> (bIsJumping)
    B="float > float"(A=Get VelocityZ!(self:Fort Player Anim Instance Object Reference=@n1) -> (VelocityZ:float), B=0)
   )
   > Set JumpUpOffsetSmooth(JumpUpOffsetSmooth:Vector=VInterpTo(
     Current=Get JumpUpOffsetSmooth() -> (JumpUpOffsetSmooth:Vector)
     Target=SelectVector(A="0,-10,25", B="0, 0, 0", bPickA=@n2)
     DeltaTime=Get DeltaTimeX()
     InterpSpeed=10
    )) -> (Output_Get:Vector)
   > Outputs!(then:Exec)
  ]
 }
 GetOffsetValues {nodes [
   n1 = Break Pet Offsets Cosmos!(PetOffsets_Cosmos:"Pet Offsets Cosmos Structure (by ref)"=Get CarrierOffsets() -> (CarrierOffsets:Pet Offsets Cosmos Structure)) -> (
    DefaultIdle_8_#1:Vector=~
    DefaultDownsights_14_#2:Vector=~
    CrouchIdle_13_#3:Vector=~
    CrouchDownsights_15_#4:Vector=~
    DBNOIdle_9_#5:Vector=~
    MeleePlaying_17_#6:Vector=~
    CrouchMeleePlaying_51_#7:Vector=~
    EmotePlaying_19_#8:Vector=~
    DefaultSkydive_22_#9:Vector=~
    ParachuteOpen_53_#10:Vector=~
    DivingSkydive_48_#11:Vector=~
    SurfaceSwimmingIdle_26_#12:Vector=~
    SurfaceSwimmingDownsights_28_#13:Vector=~
    SurfaceSwimmingMeleePlaying_32_#14:Vector=~
    Surface_Swimming_Sprint_56_#15:Vector=~
    DiveJumping_35_#16:Vector=~
    CrouchTurning_44_#17:Vector=~
   ) {is=title K2Node_BreakStruct}
   n2 = Get FortPlayerAnimBP()
   n3 = Get VelocityZ(self=@n2)
   SelectVector(A="0,-30,15", B=@n1, bPickA=AND Boolean(A=Get bIsJumping(self=@n2), B="float > float"(A=@n3, B=0)))
   Inputs!()
   n4 = Get bIsTargeting!(self:Fort Player Anim Instance Object Reference=Get FortPlayerAnimBP()) -> (bIsTargeting)
   n5 = Get bIsTargeting(self=Get FortPlayerAnimBP())
   n6 = Get FortPlayerAnimBP()
   n7 = OR Boolean(
    A=Get bTransition_LocomotionAdditive_CrouchTurning!(self:Fort Player Anim Instance Object Reference=@n6) -> (bTransition_LocomotionAdditive_CrouchTurning:bool)
    B=Get bIsMoving2D(self=@n6)
   )
   n8 = SelectVector(
    A=vector + vector!(
     A=SelectVector(A=@n1.CrouchDownsights_15_#4, B=@n1.CrouchIdle_13_#3, bPickA=@n5)
     B=SelectVector(A=@n1.CrouchTurning_44_#17, B="0, 0, 0", bPickA=@n7)
    ) -> (ReturnValue)
    B=vector + vector(A=SelectVector(A=@n1.DefaultDownsights_14_#2, B=@n1, bPickA=@n4), B=Get JumpUpOffsetSmooth())
    bPickA=Get bIsCrouching!(self:Fort Player Anim Instance Object Reference=Get FortPlayerAnimBP()) -> (bIsCrouching)
   )
   n9 = Get bIsDBNO!(self:Fort Player Anim Instance Object Reference=Get FortPlayerAnimBP()) -> (bIsDBNO)
   n10 = Get FortPlayerAnimBP()
   n11 = Get FortPlayerAnimBP()
   n12 = SelectVector(
    A=@n1.CrouchMeleePlaying_51_#7
    B=SelectVector(
     A=@n1.MeleePlaying_17_#6
     B=SelectVector(A=@n1.DBNOIdle_9_#5, B=@n8, bPickA=@n9)
     bPickA=Get bIsPlayingMeleeAnim(self=@n10)
    )
    bPickA=AND Boolean(A=Get bIsPlayingMeleeAnim(self=@n11), B=Get bIsCrouching(self=@n11))
   )
   n13 = AND Boolean(A=Get AnyEmotePlaying() -> (AnyEmotePlaying:bool), B=NOT Boolean(A=Get Emote_NonPlayerTrigger Montage Playing()))
   n14 = Get IsParachuteOpen()
   n15 = SelectVector(
    A=@n1.DivingSkydive_48_#11
    B=SelectVector(
     A=SelectVector(A=@n1.ParachuteOpen_53_#10, B=@n1.DefaultSkydive_22_#9, bPickA=@n14)
     B=SelectVector(A=@n1.EmotePlaying_19_#8, B=@n12, bPickA=@n13)
     bPickA=Get IsSkydiving()
    )
    bPickA=Get IsDiving() -> (IsDiving:bool)
   )
   n16 = Get bIsTargeting(self=Get FortPlayerAnimBP())
   n17 = Get FortPlayerAnimBP()
   n18 = SelectVector(
    A=@n1.Surface_Swimming_Sprint_56_#15
    B=SelectVector(
     A=@n1.SurfaceSwimmingMeleePlaying_32_#14
     B=SelectVector(
      A=@n1.SurfaceSwimmingDownsights_28_#13
      B=@n1.SurfaceSwimmingIdle_26_#12
      bPickA=@n16
     )
     bPickA=Get bIsPlayingMeleeAnim(self=@n17)
    )
    bPickA=Get bIsSprinting!(self:Fort Player Anim Instance Object Reference=Get FortPlayerAnimBP()) -> (bIsSprinting)
   )
   n19 = Get bIsSurfaceSwimming!(self:Fort Player Anim Instance Object Reference=Get FortPlayerAnimBP()) -> (bIsSurfaceSwimming)
   n20 = Get FortPlayerAnimBP()
   Outputs!(Default Idle:Vector=SelectVector(
     A=@n1.DiveJumping_35_#16
     B=SelectVector(A=@n18, B=@n15, bPickA=@n19)
     bPickA=Get bIsDiveJumping!(self:Fort Player Anim Instance Object Reference=@n20) -> (bIsDiveJumping:bool)
    ))
   Get bIsPlayingEmote(self=Get FortPlayerAnimBP())
  ]}
]
variables [
 AnyEmotePlaying {type=boolean}
 bCarrierDoorsClose {type=boolean}
 CarrierOffsets {
  type=PetOffsets_Cosmos
  default {
   crouch Downsights=<15 -30 -30>
   crouch Idle=<0 -30 -30>
   crouch Melee Playing=<0 -55 -60>
   crouch Turning=<5 -10 15>
   dBNO Idle=<0 10 -15>
   default Downsights=<15 -30 0>
   default Idle=<0 -30 0>
   default Skydive=<0 -30 40>
   dive Jumping=<30 -5 35>
   diving Skydive=<0 -20 50>
   emote Playing=<20 -40 -10>
   melee Playing=<0 -30 -30>
   parachute Open=<0 -70 20>
   surface Swimming Downsights=<20 -20 30>
   surface Swimming Idle=<20 -20 30>
   surface Swimming Melee Playing=<0 -25 -20>
   surface_Swimming_Sprint=<30 -30 40>
  }
 }
 CarrierPackOffsetRate {type=Vector}
 CarrierPackOffsetRateSmooth {type=Vector}
 CarrierPackOffsetRotationReaction {type=Rotator}
 CarrierPackOffsetSpringState {type=VectorSpringState, default {}}
 CarrierPackOffsetTotal {type=Vector}
 CarrierPackOffsetTotalLastTick {type=Vector}
 CarrierPackRotationSpringState {type=FloatSpringState, default {}}
 CarrierPackVelocityRotationReaction {type=Rotator}
 Damage_Taken {type=boolean}
 DeltaTimeX {type=number}
 Emote_NonPlayerTrigger EID Item {type=FortItemDefinition, default=/BRCosmetics/Athena/Items/Cosmetics/Dances/EID_CosmosPet}
 Emote_NonPlayerTrigger Montage {type=AnimMontage, default=/Game/Animation/Game/MainPlayer/Emotes/NonPlayerTrigger/Emote_NonPlayerTrigger_CMM_M}
 Emote_NonPlayerTrigger Montage Playing {type=boolean}
 FortPlayerAnimBP {type=Fortnite_M_Avg_Player_AnimBlueprint_C, default=null}
 FortPlayerLocalVelocity {type=Vector}
 FortPlayerPawn {type=FortPlayerPawn}
 Is DBNOCarried {type=boolean}
 IsDBNOCarriedAlpha {type=number}
 IsDiving {type=boolean}
 IsParachuteJustOpened {type=boolean}
 IsParachuteOpen {type=boolean}
 IsParachuteOpenLastTick {type=boolean}
 IsSkydiving {type=boolean}
 JumpUpOffsetSmooth {type=Vector}
 ParachuteJustOpenedOffset {type=Vector}
 PetAnimBP {type=M_MED_Cosmos_Pet_AnimBP_C, default=null}
 PetInstCosmos {type=B_PetInst_Cosmos_C, default=null}
 RootRotationSpring {type=number}
 RootYawOffset {type=number}
 RootYawOffsetLastTick {type=number}
 RootYawOffsetSmooth {type=number}
]
